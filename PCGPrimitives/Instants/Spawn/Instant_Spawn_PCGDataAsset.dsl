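; PCG graph that spawns the contents of a referenced PCG Data Asset at given transforms, under a new labeled, unattached target actor.
; It loads the asset, keeps data tagged Points and copies those points onto the spawn transforms.
; Points with Mesh feed static mesh instancing; those with SkeletalMesh feed static-rendered instanced skinned meshes, both supporting material overrides.
; A branch on the asset reference prints an error when it is missing.
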(graph Instant_Spawn_PCGDataAsset
  (params (PCGDataAssetReference :type (SoftObject /Script/PCG.PCGDataAsset)) (SpawnPointTransforms :type Transform) (ActorLabel :type String :default "PCGAssembly"))
  (bind AttributeCompareOp_0 (Attribute%20Compare%20Op :InA PCGDataAssetReference))
  (bind Branch_0 (Branch :In PCGDataAssetReference))
  (bind CreatePoints_1 (Create%20Points :Transform SpawnPointTransforms))
  (bind PCGLoadDataAsset_0 (Load%20PCG%20Data%20Asset :bLoadFromInput true))
  (bind FilterDataByTag_0 (Filter%20Data%20By%20Tag :SelectedTags "Points"))
  (bind ToPoint_0 (To%20Point))
  (bind CopyPoints_0 (Copy%20Points))
  (bind CreateTargetActor_1 (Create%20Target%20Actor :AttachOptions NotAttached :TemplateActorClass /Script/Engine.Actor :ActorLabel-param ActorLabel :ActorPivot-param SpawnPointTransforms))
  (bind AttributeFilter_2 (Filter%20Attribute%20Elements :AttributeTypes (struct :Type SoftObjectPath) :bGenerateOutputDataEvenIfEmpty false :bUseConstantThreshold true :bWarnOnDataMissingAttribute false :Operator NotEqual :TargetAttribute Mesh))
  (bind StaticMeshSpawner_0 (Static%20Mesh%20Spawner :bApplyMeshBoundsToPoints false :MeshSelectorParameters (class:/Script/PCG.PCGMeshSelectorByAttribute :AttributeName Mesh :bUseAttributeMaterialOverrides true :MaterialOverrideAttributes [Material] :TemplateDescriptor (struct :bIncludeInHLOD false :bUseDefaultCollision true :bGenerateOverlapEvents true)) :MeshSelectorType /Script/PCG.PCGMeshSelectorByAttribute))
  (bind AttributeFilter_3 (Filter%20Attribute%20Elements :AttributeTypes (struct :Type SoftObjectPath) :bGenerateOutputDataEvenIfEmpty false :bUseConstantThreshold true :bWarnOnDataMissingAttribute false :Operator NotEqual :TargetAttribute SkeletalMesh :ThresholdAttribute @Last))
  (bind AddAttribute_0 (Add%20Attribute :AttributeTypes (struct :Type SoftObjectPath) :OutputTarget AnimBank))
  (bind AddAttribute_1 (Add%20Attribute :AttributeTypes (struct :Type Integer32 :Int32Value -1) :OutputTarget SequenceIndex))
  (bind SkinnedMeshSpawner_0 (Instanced%20Skinned%20Mesh%20Spawner :MeshSelectorParameters (class:/Script/PCG.PCGSkinnedMeshSelector :bUseAttributeMaterialOverrides true :MaterialOverrideAttributes [Material] :MeshAttribute SkeletalMesh :TemplateDescriptor (struct :bRenderStatic true :BodyInstance (struct :CollisionProfileName BlockAll)))))
  (bind Print%20String_0 (Print%20String :bPrefixWithNode false :bPrintPerComponent false :PrintString "Missing PCG Data Asset reference" :Verbosity Error))
  (connect AttributeFilter_3 InsideFilter AddAttribute_0 In)
  (connect AddAttribute_0 Out AddAttribute_1 In)
  (connect CopyPoints_0 Out AttributeFilter_2 In)
  (connect CopyPoints_0 Out AttributeFilter_3 In)
  (connect AttributeCompareOp_0 Out Branch_0 bOutputToB)
  (connect ToPoint_0 Out CopyPoints_0 Source)
  (connect CreatePoints_1 Out CopyPoints_0 Target)
  (connect Branch_0 Output%20A CreatePoints_1 Execution%20Dependency)
  (connect Branch_0 Output%20A CreateTargetActor_1 Execution%20Dependency)
  (connect PCGLoadDataAsset_0 Out FilterDataByTag_0 In)
  (connect Branch_0 Output%20A PCGLoadDataAsset_0 In)
  (connect Branch_0 Output%20B Print%20String_0 In)
  (connect AddAttribute_1 Out SkinnedMeshSpawner_0 In)
  (connect CreateTargetActor_1 Out SkinnedMeshSpawner_0 TargetActor)
  (connect AttributeFilter_2 InsideFilter StaticMeshSpawner_0 In)
  (connect CreateTargetActor_1 Out StaticMeshSpawner_0 TargetActor)
  (connect FilterDataByTag_0 InsideFilter ToPoint_0 In))
